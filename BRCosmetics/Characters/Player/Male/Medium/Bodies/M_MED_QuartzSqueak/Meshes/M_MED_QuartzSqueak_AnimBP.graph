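# Post-copy animation Blueprint for the QuartzSqueak male body cosmetic, adjusting the pose copied from its parent mesh.
# A valid ExtraPart "shoe" mesh selects the plain copied pose; otherwise a Control Rig, fed emote state, processes body pose layered with face-mesh bones from faceAttach.
# The head then gets a scaled per-axis counter-rotation against its neck_02-relative delta.
# A knee-piece bone gets a small fixed offset, and base_Brain is constrained between neck_01 and head.
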AnimBlueprint /BRCosmetics/Characters/Player/Male/Medium/Bodies/M_MED_QuartzSqueak/Meshes/M_MED_QuartzSqueak_AnimBP
target_skeleton=&/M_MED_QuartzSqueak_Skeleton
graphs [
 AnimGraph {nodes [
   n1 = Control Rig!(Source=Use cached pose 'StartCache'(), Alpha, emoting:bool=Get bIsPlayingEmote()) -> (Pose) {controlRigAssetReference {blueprintRigClass=&/M_MED_QuartzSqueak_CtrlRig.M_MED_QuartzSqueak_CtrlRig_C}}
   n2 = Copy Pose From Mesh() {bUseAttachedParent=true}
   n3 = Get Shoes_Active() -> (Shoes_Active:bool)
   n4 = "Transform (Modify) Bone - Bone: head"(
    ComponentPose=Local To Component(LocalPose=Blend Poses by bool(BlendPose_0=@n2, BlendPose_1=@n1, bActiveValue=@n3) {blendType=Linear})
    Alpha=1
    Translation="0,0,0"
    Rotation=Get head rot() -> (head rot:Rotator)
    Scale="1,1,1"
   ) {rotationMode=BMM_Additive, rotationSpace=BCS_BoneSpace, boneToModify {boneName=head}, rotation=<rot 50 50 50>}
   n5 = "Transform (Modify) Bone - Bone: dyn_dog_metalKneePiece"!(ComponentPose=@n4, bAlphaBoolEnabled:bool=True, Translation="0,-0.25,0", Rotation="0,0,2", Scale="1,1,1") -> (Pose) {
    alphaInputType=Bool
    rotationMode=BMM_Additive
    rotationSpace=BCS_BoneSpace
    translationMode=BMM_Additive
    translationSpace=BCS_BoneSpace
    boneToModify {boneName=dyn_dog_metalKneePiece}
    rotation {roll=2}
    translation {y=-0.25}
   }
   n6 = "Constraint - base_Brain (PP)"!(ComponentPose=@n5, Alpha=0.5, ConstraintWeights_0=0.5, ConstraintWeights_1:float=0.5) -> (Pose) {boneToModify {boneName=base_Brain}, constraintSetup [{targetBone {boneName=neck_01}}, {targetBone {boneName=head}}], constraintWeights [1,1]}
   Output Pose(Result=Component To Local(ComponentPose=@n6)) {layerGroup=None}
   n7 = "Get Face Mesh "() -> ("Face Mesh ":Skeletal Mesh Component Object Reference)
   Save cached pose 'StartCache'(Pose=Layered blend per bone(
     BasePose=Copy Pose From Mesh() {bUseAttachedParent=true}
     BlendPoses_0=Copy Pose From Mesh(SourceMeshComponent=@n7) {bUseAttachedParent=true}
    ) {layerSetup [{branchFilters [{boneName=faceAttach}]}]}) {cacheName=StartCache}
  ]}
 EventGraph {nodes [
   Event BlueprintUpdateAnimation()
   > n1 = Sequence()
   n2 = GetOwnerSkeletalMeshComponent()
   n3 = BreakRotator(InRot=BreakTransform(InTransform=GetDeltaTransformFromRefPose(self=@n2, BoneName=head, BaseName=neck_02)).Rotation)
   n4 = Get Y percentage() -> (Y percentage:float)
   n5 = Get X percentage() -> (X percentage:float)
   n6 = Get Z percentage() -> (Z percentage:float)
   > Set head rot(head rot:Rotator=MakeRotator(
     Roll=float * float!(A=@n3, B=-1, C:double=@n5) -> (ReturnValue)
     Pitch=float * float(A=@n3.Pitch, B=-1, C=@n4)
     Yaw=float * float(A=@n3.Yaw, B=-1, C=@n6)
    )) -> (Output_Get:Rotator)
   n7 = Get!(execute:Exec=@n1.then_1) -> (then:Exec, else:Exec, ShoeMesh:Skeletal Mesh Component Budgeted Object Reference)
   > n8 = Is Valid(InputObject=Get SkeletalMesh!(self:Skinned Mesh Component Object Reference=@n7) -> (SkeletalMesh:Skeletal Mesh Object Reference))
   Set Shoes_Active(execute=[@n8.Is Not Valid, @n7.else], Shoes_Active:bool=false) -> (Output_Get:bool)
   Set Shoes_Active(execute=@n8.Is Valid, Shoes_Active=true)
   n9 = Event OnCharacterCustomizationLoaded()
   > n10 = Cast To PlayerPawn_Athena(Object=TryGetPawnOwner())
   > "Set Body Mesh "("Body Mesh ":Skeletal Mesh Component Object Reference=GetSkeletalMeshForPartType(self=@n10)) -> (Output_Get:Skeletal Mesh Component Object Reference)
   > "Set Face Mesh "("Face Mesh ":Skeletal Mesh Component Object Reference=GetSkeletalMeshForPartType(self=@n10, PartType=Head)) -> (Output_Get:Skeletal Mesh Component Object Reference)
   > Set ShoeMesh(ShoeMesh:Skeletal Mesh Component Budgeted Object Reference=Get Array Item with EFort Custom Part Type(Array=@n9.CustomMeshParts, PartType=ExtraPart)) -> (Output_Get:Skeletal Mesh Component Budgeted Object Reference)
  ]}
]
variables [
 "Body Mesh " {type=SkeletalMeshComponent}
 "Face Mesh " {type=SkeletalMeshComponent}
 head rot {type=Rotator}
 ShoeMesh {type=SkeletalMeshComponentBudgeted}
 Shoes_Active {type=boolean}
 X percentage {type=number, default=0.6}
 Y percentage {type=number, default=0.85}
 Z percentage {type=number, default=0.95}
]
