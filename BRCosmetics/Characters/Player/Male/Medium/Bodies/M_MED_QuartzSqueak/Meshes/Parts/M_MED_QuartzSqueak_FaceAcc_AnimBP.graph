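# Post-process style AnimBP for QuartzSqueak's face accessory, animating ears and backpack tubes over the parent mesh's copied pose.
# Chained AnimDynamics springs swing four ear bones under fixed reduced gravity; the computed Gravity Override variable is not visibly wired to them.
# Additive transforms offset the tube bone, rotating and shifting it with spine bend measured against ref pose.
# Switching from skydive float to dive adds an interpolated, delay-reset translation kick.
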AnimBlueprint /BRCosmetics/Characters/Player/Male/Medium/Bodies/M_MED_QuartzSqueak/Meshes/Parts/M_MED_QuartzSqueak_FaceAcc_AnimBP
target_skeleton=&/M_MED_QuartzSqueak_FaceAcc_Skeleton
graphs [
 AnimGraph {nodes [
   n1 = Anim Dynamics - Bone: dyn_ear_l_01!(
    ComponentPose=Local To Component(LocalPose=Copy Pose From Mesh() {bUseAttachedParent=true})
    Alpha=1
    GravityOverride:Vector="0,0,-980"
    ExternalForce="0,0,0"
   ) -> (Pose) {
    angularDampingOverride=1
    angularSpringConstant=20000
    bAngularSpring=true
    bOverrideAngularDamping=true
    bUseGravityOverride=true
    gravityScale=0.35
    simulationSpace=RootRelative
    boundBone {boneName=dyn_ear_l_01}
    componentLinearAccScale=<1 1 1>
    componentLinearVelScale=<1 1 1>
    gravityOverride {z=-980}
    physicsBodyDefinitions [{
      boundBone {boneName=dyn_ear_l_01}
      boxExtents=<20 30 30>
      constraintSetup {angularLimitsMax=<15 15 15>, angularLimitsMin=<-15 -5 -15>, angularTarget=<1 0 0>}
      localJointOffset=<4 0 0>
     }]
   }
   n2 = Anim Dynamics - Bone: dyn_ear_l_02!(ComponentPose=@n1, Alpha=1, GravityOverride:Vector="0,0,-980", ExternalForce="0,0,0") -> (Pose) {
    angularDampingOverride=1
    angularSpringConstant=20000
    bAngularSpring=true
    bOverrideAngularDamping=true
    bUseGravityOverride=true
    gravityScale=0.35
    simulationSpace=RootRelative
    boundBone {boneName=dyn_ear_l_02}
    componentLinearAccScale=<1 1 1>
    componentLinearVelScale=<1 1 1>
    gravityOverride {z=-980}
    physicsBodyDefinitions [{
      boundBone {boneName=dyn_ear_l_02}
      boxExtents=<20 30 30>
      constraintSetup {angularLimitsMax=<15 8 15>, angularLimitsMin=<-15 -8 -15>, angularTarget=<1 0 0>}
      localJointOffset=<4 0 0>
     }]
   }
   n3 = Anim Dynamics - Bone: dyn_ear_r_01!(ComponentPose=@n2, Alpha=1, GravityOverride:Vector="0,0,-980", ExternalForce="0,0,0") -> (Pose) {
    angularDampingOverride=1
    angularSpringConstant=20000
    bAngularSpring=true
    bOverrideAngularDamping=true
    bUseGravityOverride=true
    gravityScale=0.35
    simulationSpace=RootRelative
    boundBone {boneName=dyn_ear_r_01}
    componentLinearAccScale=<1 1 1>
    componentLinearVelScale=<1 1 1>
    gravityOverride {z=-980}
    physicsBodyDefinitions [{
      boundBone {boneName=dyn_ear_r_01}
      boxExtents=<20 30 30>
      constraintSetup {angularLimitsMax=<15 15 15>, angularLimitsMin=<-15 -5 -15>, angularTarget=<1 0 0>}
      localJointOffset=<-4 0 0>
     }]
   }
   n4 = Anim Dynamics - Bone: dyn_ear_r_02!(ComponentPose=@n3, Alpha=1, GravityOverride:Vector="0,0,-980", ExternalForce="0,0,0") -> (Pose) {
    angularDampingOverride=1
    angularSpringConstant=20000
    bAngularSpring=true
    bOverrideAngularDamping=true
    bUseGravityOverride=true
    gravityScale=0.35
    simulationSpace=RootRelative
    boundBone {boneName=dyn_ear_r_02}
    componentLinearAccScale=<1 1 1>
    componentLinearVelScale=<1 1 1>
    gravityOverride {z=-980}
    physicsBodyDefinitions [{
      boundBone {boneName=dyn_ear_r_02}
      boxExtents=<20 30 30>
      constraintSetup {angularLimitsMax=<15 5 15>, angularLimitsMin=<-15 -5 -15>, angularTarget=<1 0 0>}
      localJointOffset=<-4 0 0>
     }]
   }
   n5 = "Transform (Modify) Bone - Bone: dyn_packtubes"(ComponentPose=@n4, Alpha=1, Translation="0,-2,0", Rotation="0,0,0", Scale="1,1,1") {translationMode=BMM_Additive, translationSpace=BCS_BoneSpace, boneToModify {boneName=dyn_packtubes}, translation {y=-2}}
   n6 = "Transform (Modify) Bone - Bone: dyn_packtubes"(ComponentPose=@n5, Alpha=1, Translation=Get backpack_translation(), Rotation=Get backpack_rotate(), Scale="1,1,1") {
    rotationMode=BMM_Additive
    rotationSpace=BCS_BoneSpace
    translationMode=BMM_Additive
    translationSpace=BCS_BoneSpace
    alphaBoolBlend {blendOption=Linear}
    boneToModify {boneName=dyn_packtubes}
   }
   Output Pose(Result=Component To Local(ComponentPose=@n6)) {layerGroup=None}
  ]}
 EventGraph {nodes [
   n1 = Event BlueprintUpdateAnimation()
   > Set Delta Time X(Delta Time X=@n1.DeltaTimeX)
   > n2 = Sequence!(execute) -> (then_0, then_1, then_2:Exec)
   n3 = Branch(execute=@n2.then_1, Condition=Get bIsSkydiveFloating())
   > Set Skydive_test(Skydive_test:bool=true) -> (Output_Get:bool)
   Branch(execute=@n3.else, Condition=Get bIsSkydiveDiveMode())
   > n4 = Branch(Condition=Get Skydive_test() -> (Skydive_test:bool))
   > Set Skydive_Translation_y(Skydive_Translation_y:double=FInterpTo(Current=Get Skydive_Translation_y() -> (Skydive_Translation_y:double), Target=-15, DeltaTime=Get Delta Time X(), InterpSpeed=10)) -> (Output_Get:double)
   > Set Skydive_Translation_z(Skydive_Translation_z:double=-2) -> (Output_Get:double)
   > Delay(Duration=0.45)
   > Set Skydive_Translation_y(Skydive_Translation_y=0)
   > Set Skydive_Translation_z(Skydive_Translation_z=0)
   > Set Skydive_test(Skydive_test=false)
   Set Skydive_Translation_y(execute=@n4.else, Skydive_Translation_y=0)
   > Set Skydive_Translation_z(Skydive_Translation_z=0)
   n5 = GetOwnerSkeletalMeshComponent()
   n6 = BreakRotator(InRot=BreakTransform(InTransform=GetDeltaTransformFromRefPose(self=@n5, BoneName=spine_05, BaseName=spine_01)).Rotation)
   Set backpack_rotate(execute=@n2.then_2, backpack_rotate=MakeRotator(Roll=MapRangeClamped(Value=@n6.Yaw, InRangeA=0, InRangeB=-30, OutRangeB=35)))
   n7 = GetDeltaTransformFromRefPose(self=GetOwnerSkeletalMeshComponent(), BoneName=spine_05, BaseName=spine_01)
   n8 = MapRangeClamped(Value=BreakRotator(InRot=BreakTransform(InTransform=@n7).Rotation).Yaw, InRangeA=0, InRangeB=-15, OutRangeB=-2)
   n9 = Get Skydive_Translation_y()
   > Set y_Translation(y_Translation:double=FInterpTo(Current=Get y_Translation() -> (y_Translation:double), Target=float + float(A=@n8, B=@n9), DeltaTime=Get Delta Time X(), InterpSpeed=12)) -> (Output_Get:double)
   > Set z_Translation(z_Translation:double=FInterpTo(
     Current=Get z_Translation() -> (z_Translation:double)
     Target=Get Skydive_Translation_z() -> (Skydive_Translation_z:double)
     DeltaTime=Get Delta Time X()
     InterpSpeed=12
    )) -> (Output_Get:double)
   n10 = Get bIsSkydiveDiveMode()
   n11 = Get y_Translation()
   n12 = Get bIsSkydiveDiveMode()
   n13 = Get z_Translation()
   > Set backpack_translation(backpack_translation=MakeVector(X=SelectFloat(A=@n13, B=0, bPickA=@n12), Y=SelectFloat(A=@n11, B=@n8, bPickA=@n10)))
   n14 = "Gravity Override State Function (Parameter Asset) (Nativized)"(
    execute=@n2.then_0
    CharacterPartAnimInstance=Self-Reference()
    GameTimeInSeconds=GetGameTimeInSeconds()
    Parameters=&/M_MED_QuartzSqueak_FaceAcc_Parameters
   )
   > Set Gravity Override(Gravity Override=@n14)
  ]}
]
variables [
 backpack_rotate {type=Rotator}
 backpack_translation {type=Vector}
 Delta Time X {type=number}
 Gravity Override {type=Vector}
 Skydive_test {type=boolean}
 Skydive_Translation_y {type=number}
 Skydive_Translation_z {type=number}
 y_Translation {type=number}
 z_Translation {type=number}
]
